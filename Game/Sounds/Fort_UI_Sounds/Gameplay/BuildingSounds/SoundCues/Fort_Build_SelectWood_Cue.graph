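# Sound cue for the wood selection sound in Fortnite's building gameplay UI.
# A single wave player feeds the PieceFit_13 sound through a Modulator node at reduced cue volume.
# It plays non-spatialized on the critical in-game UI sound class.
# An overridden concurrency rule blocks new instances rather than cutting off playing ones.
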SoundCue /Game/Sounds/Fort_UI_Sounds/Gameplay/BuildingSounds/SoundCues/Fort_Build_SelectWood_Cue
root=Modulator_0
nodes [
 {id=Modulator_0, children [WavePlayer_6]}
 {id=WavePlayer_6, properties {soundWaveAssetPtr=/Game/Sounds/Fort_UI_Sounds/Gameplay/Fort_Build_PieceFit_13}}
]
properties {
 bOverrideConcurrency=true
 maxDistance=2097152
 soundClassObject=&Fort_UI_InGame_Critical
 volumeMultiplier=0.45
 attenuationOverrides {bSpatialize=false, falloffDistance=2000, nonSpatializedRadiusStart=1000}
 concurrencyOverrides {resolutionRule=PreventNew}
}
