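# Sound cue for an enemy starting a level 1 metal build, layering an initiation hit with a looping construction sound.
# The hit randomly picks one of three "construction done" metal waves, pitched up, with enemy-initiate attenuation.
# The quieter loop starts after a short delay with wall attenuation and a non-critical sound class, switching between standard and optimized waves by quality level.
# A player wall-build concurrency set governs simultaneous instances.
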SoundCue /Game/Sounds/Fort_UI_Sounds/Gameplay/BuildingSounds/SoundCues/Fort_Build_StartMetal_Lev1_Enemy_Cue
root=Mixer_4
nodes [
 {id=Attenuation_0, children [Random_1], properties {attenuationSettings=/Game/Sounds/Attenuation/Fort_Build_Initiate_Enemy_Att}}
 {id=Attenuation_1, children [QualityLevel_1], properties {attenuationSettings=/Game/Sounds/Attenuation/Fort_Build_Wall_Att}}
 {id=Delay_4, children [Modulator_10], properties {delayMax=0.33, delayMin=0.3}}
 {id=Mixer_4, children [Modulator_9,Delay_4], properties {inputVolume [1,0.6]}}
 {id=Modulator_10, children [SoundClass_0]}
 {id=Modulator_9, children [SoundClass_1], properties {pitchMax=1.4, pitchMin=1.3, volumeMax=1, volumeMin=1}}
 {id=QualityLevel_1, children [WavePlayer_8,WavePlayer_8,WavePlayer_1,WavePlayer_1]}
 {id=Random_1, children [WavePlayer_3,WavePlayer_4,WavePlayer_5], properties {weights [1,1,1]}}
 {
  id=SoundClass_0
  children [Attenuation_1]
  properties {soundClassOverride=/Game/Sounds/FortSoundClassesAndModes/Fort_Player_Construction_NonCritical}
 }
 {id=SoundClass_1, children [Attenuation_0], properties {soundClassOverride=/Game/Sounds/FortSoundClassesAndModes/Fort_Player_Construction}}
 {
  id=WavePlayer_1
  properties {bLooping=true, soundWaveAssetPtr=/Game/Sounds/Fort_UI_Sounds/Gameplay/BuildingSounds/Wavs/metal_level_const_level_01_A_opt}
 }
 {id=WavePlayer_3, properties {soundWaveAssetPtr=/Game/Sounds/Fort_UI_Sounds/Gameplay/BuildingSounds/Wavs/Fort_Construction_Done_Metal_01}}
 {id=WavePlayer_4, properties {soundWaveAssetPtr=/Game/Sounds/Fort_UI_Sounds/Gameplay/BuildingSounds/Wavs/Fort_Construction_Done_Metal_02}}
 {id=WavePlayer_5, properties {soundWaveAssetPtr=/Game/Sounds/Fort_UI_Sounds/Gameplay/BuildingSounds/Wavs/Fort_Construction_Done_Metal_03}}
 {
  id=WavePlayer_8
  properties {bLooping=true, soundWaveAssetPtr=/Game/Sounds/Fort_UI_Sounds/Gameplay/BuildingSounds/Wavs/metal_level_const_level_01_A}
 }
]
properties {
 maxDistance=3584
 volumeMultiplier=1
 attenuationOverrides {falloffDistance=2600}
 concurrencyOverrides {resolutionRule=PreventNew}
 concurrencySet [/Game/Sounds/Concurrency_Rules/Player_Wall_Build_Conc]
}
