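# Sound cue for an enemy starting level 2 metal building, mixing a one-shot hit with a construction loop.
# The hit randomly picks one of three metal "construction done" waves, pitched up, using enemy build-initiate attenuation.
# The quieter loop follows after a short delay with wall-build attenuation and a NonCritical sound class, switching between standard and OPT waves by quality level.
# A shared wall-build concurrency set governs simultaneous instances.
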SoundCue /Game/Sounds/Fort_UI_Sounds/Gameplay/BuildingSounds/SoundCues/Fort_Build_StartMetal_Lev2_Enemy_Cue
root=Mixer_7
nodes [
 {id=Attenuation_0, children [Random_1], properties {attenuationSettings=/Game/Sounds/Attenuation/Fort_Build_Initiate_Enemy_Att}}
 {id=Attenuation_1, children [QualityLevel_0], properties {attenuationSettings=/Game/Sounds/Attenuation/Fort_Build_Wall_Att}}
 {id=Delay_7, children [Modulator_16], properties {delayMax=0.33, delayMin=0.3}}
 {id=Mixer_7, children [Modulator_15,Delay_7], properties {inputVolume [0.8,0.6]}}
 {id=Modulator_15, children [SoundClass_1], properties {pitchMax=1.4, pitchMin=1.3, volumeMax=1, volumeMin=1}}
 {id=Modulator_16, children [SoundClass_0]}
 {id=QualityLevel_0, children [WavePlayer_16, WavePlayer_16, WavePlayer_0, WavePlayer_0]}
 {id=Random_1, children [WavePlayer_3,WavePlayer_4,WavePlayer_5], properties {weights [1,1,1]}}
 {
  id=SoundClass_0
  children [Attenuation_1]
  properties {soundClassOverride=/Game/Sounds/FortSoundClassesAndModes/Fort_Player_Construction_NonCritical}
 }
 {id=SoundClass_1, children [Attenuation_0], properties {soundClassOverride=/Game/Sounds/FortSoundClassesAndModes/Fort_Player_Construction}}
 {
  id=WavePlayer_0
  properties {bLooping=true, soundWaveAssetPtr=/Game/Sounds/Fort_UI_Sounds/Gameplay/BuildingSounds/Wavs/metal_level_const_level_02_A_OPT}
 }
 {
  id=WavePlayer_16
  properties {bLooping=true, soundWaveAssetPtr=/Game/Sounds/Fort_UI_Sounds/Gameplay/BuildingSounds/Wavs/metal_level_const_level_02_A}
 }
 {id=WavePlayer_3, properties {soundWaveAssetPtr=/Game/Sounds/Fort_UI_Sounds/Gameplay/BuildingSounds/Wavs/Fort_Construction_Done_Metal_01}}
 {id=WavePlayer_4, properties {soundWaveAssetPtr=/Game/Sounds/Fort_UI_Sounds/Gameplay/BuildingSounds/Wavs/Fort_Construction_Done_Metal_02}}
 {id=WavePlayer_5, properties {soundWaveAssetPtr=/Game/Sounds/Fort_UI_Sounds/Gameplay/BuildingSounds/Wavs/Fort_Construction_Done_Metal_03}}
]
properties {
 maxDistance=3584
 volumeMultiplier=1
 attenuationOverrides {falloffDistance=2600}
 concurrencyOverrides {resolutionRule=PreventNew}
 concurrencySet [/Game/Sounds/Concurrency_Rules/Player_Wall_Build_Conc]
}
