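# Sound cue for an enemy starting a level 3 metal build, mixing a start hit with a construction loop.
# The hit randomly picks one of three metal "construction done" waves, pitched up, with enemy-initiate attenuation.
# The quieter loop follows after a short delay, using wall attenuation, a non-critical sound class and a quality-level switch between standard and OPT level 02 waves.
# A shared wall-build concurrency set is assigned.
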SoundCue /Game/Sounds/Fort_UI_Sounds/Gameplay/BuildingSounds/SoundCues/Fort_Build_StartMetal_Lev3_Enemy_Cue
root=Mixer_8
nodes [
 {id=Attenuation_0, children [Random_0], properties {attenuationSettings=/Game/Sounds/Attenuation/Fort_Build_Initiate_Enemy_Att}}
 {id=Attenuation_1, children [QualityLevel_0], properties {attenuationSettings=/Game/Sounds/Attenuation/Fort_Build_Wall_Att}}
 {id=Delay_8, children [SoundClass_0], properties {delayMax=0.33, delayMin=0.3}}
 {id=Mixer_8, children [Modulator_17,Delay_8], properties {inputVolume [0.8,0.6]}}
 {id=Modulator_17, children [SoundClass_1], properties {pitchMax=1.4, pitchMin=1.3, volumeMax=1, volumeMin=1}}
 {id=Modulator_18, children [Attenuation_1]}
 {id=QualityLevel_0, children [WavePlayer_18, WavePlayer_18, WavePlayer_3, WavePlayer_3]}
 {id=Random_0, children [WavePlayer_0,WavePlayer_1,WavePlayer_2], properties {weights [1,1,1]}}
 {
  id=SoundClass_0
  children [Modulator_18]
  properties {soundClassOverride=/Game/Sounds/FortSoundClassesAndModes/Fort_Player_Construction_NonCritical}
 }
 {id=SoundClass_1, children [Attenuation_0], properties {soundClassOverride=/Game/Sounds/FortSoundClassesAndModes/Fort_Player_Construction}}
 {id=WavePlayer_0, properties {soundWaveAssetPtr=/Game/Sounds/Fort_UI_Sounds/Gameplay/BuildingSounds/Wavs/Fort_Construction_Done_Metal_01}}
 {id=WavePlayer_1, properties {soundWaveAssetPtr=/Game/Sounds/Fort_UI_Sounds/Gameplay/BuildingSounds/Wavs/Fort_Construction_Done_Metal_02}}
 {
  id=WavePlayer_18
  properties {bLooping=true, soundWaveAssetPtr=/Game/Sounds/Fort_UI_Sounds/Gameplay/BuildingSounds/Wavs/metal_level_const_level_02_A}
 }
 {id=WavePlayer_2, properties {soundWaveAssetPtr=/Game/Sounds/Fort_UI_Sounds/Gameplay/BuildingSounds/Wavs/Fort_Construction_Done_Metal_03}}
 {
  id=WavePlayer_3
  properties {bLooping=true, soundWaveAssetPtr=/Game/Sounds/Fort_UI_Sounds/Gameplay/BuildingSounds/Wavs/metal_level_const_level_02_A_OPT}
 }
]
properties {
 maxDistance=3584
 volumeMultiplier=1
 attenuationOverrides {falloffDistance=2600}
 concurrencyOverrides {resolutionRule=PreventNew}
 concurrencySet [/Game/Sounds/Concurrency_Rules/Player_Wall_Build_Conc]
}
