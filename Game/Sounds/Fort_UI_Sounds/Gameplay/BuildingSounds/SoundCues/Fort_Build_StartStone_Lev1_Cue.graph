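# Sound cue for the start of level 1 stone building, mixing a one-shot initiation hit with a looping construction bed.
# The hit randomly picks one of three reused wood construction-done waves and pitches it up.
# The rock construction loop enters after a short randomized delay.
# Each layer has its own attenuation and sound class, with the loop classed as non-critical, and the cue uses the player wall-build concurrency set.
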SoundCue /Game/Sounds/Fort_UI_Sounds/Gameplay/BuildingSounds/SoundCues/Fort_Build_StartStone_Lev1_Cue
root=Mixer_3
nodes [
 {id=Attenuation_0, children [Random_0], properties {attenuationSettings=/Game/Sounds/Attenuation/Fort_Build_Initiate_Att}}
 {id=Attenuation_1, children [Modulator_1], properties {attenuationSettings=/Game/Sounds/Attenuation/Fort_Build_Wall_Att}}
 {id=Delay_0, children [Mixer_2], properties {delayMax=0.33, delayMin=0.3}}
 {id=Mixer_2, children [SoundClass_0], properties {inputVolume [0.7]}}
 {id=Mixer_3, children [Modulator_0,Delay_0], properties {inputVolume [0.8,0.7]}}
 {id=Modulator_0, children [SoundClass_1], properties {pitchMax=1.4, pitchMin=1.3, volumeMax=0.6, volumeMin=0.6}}
 {id=Modulator_1, children [WavePlayer_6]}
 {id=Random_0, children [WavePlayer_0,WavePlayer_1,WavePlayer_2], properties {bShouldExcludeFromBranchCulling=true, weights [1,1,1]}}
 {
  id=SoundClass_0
  children [Attenuation_1]
  properties {soundClassOverride=/Game/Sounds/FortSoundClassesAndModes/Fort_Player_Construction_NonCritical}
 }
 {id=SoundClass_1, children [Attenuation_0], properties {soundClassOverride=/Game/Sounds/FortSoundClassesAndModes/Fort_Player_Construction}}
 {id=WavePlayer_0, properties {soundWaveAssetPtr=/Game/Sounds/Fort_UI_Sounds/Gameplay/BuildingSounds/Wavs/Fort_Construction_Done_Wood_01}}
 {id=WavePlayer_1, properties {soundWaveAssetPtr=/Game/Sounds/Fort_UI_Sounds/Gameplay/BuildingSounds/Wavs/Fort_Construction_Done_Wood_02}}
 {id=WavePlayer_2, properties {soundWaveAssetPtr=/Game/Sounds/Fort_UI_Sounds/Gameplay/BuildingSounds/Wavs/Fort_Construction_Done_Wood_03}}
 {
  id=WavePlayer_6
  properties {bLooping=true, soundWaveAssetPtr=/Game/Sounds/Fort_UI_Sounds/Gameplay/BuildingSounds/Wavs/rock_lev_01_construction_loop_B}
 }
]
properties {
 maxDistance=2560
 volumeMultiplier=0.7
 attenuationOverrides {falloffDistance=2600}
 concurrencyOverrides {resolutionRule=PreventNew}
 concurrencySet [/Game/Sounds/Concurrency_Rules/Player_Wall_Build_Conc]
}
