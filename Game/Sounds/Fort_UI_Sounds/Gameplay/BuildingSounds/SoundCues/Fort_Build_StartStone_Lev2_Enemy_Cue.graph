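# Sound cue for an enemy starting a level 2 stone build, mixing an initial one-shot with a construction loop.
# The one-shot randomly picks one of three wood construction-done waves, pitched up, under the enemy initiate attenuation and Construction sound class.
# The rock loop enters after a short delay at raised pitch, using wall attenuation and the NonCritical class, with quality levels choosing the standard or opt wave.
# A shared wall-build concurrency set is assigned.
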SoundCue /Game/Sounds/Fort_UI_Sounds/Gameplay/BuildingSounds/SoundCues/Fort_Build_StartStone_Lev2_Enemy_Cue
root=Mixer_1
nodes [
 {id=Attenuation_0, children [Random_0], properties {attenuationSettings=/Game/Sounds/Attenuation/Fort_Build_Initiate_Enemy_Att}}
 {id=Attenuation_1, children [QualityLevel_0], properties {attenuationSettings=/Game/Sounds/Attenuation/Fort_Build_Wall_Att}}
 {id=Delay_1, children [Modulator_7], properties {delayMax=0.33, delayMin=0.3}}
 {id=Mixer_1, children [SoundClass_1,SoundClass_0], properties {inputVolume [0.8,0.7]}}
 {id=Modulator_6, children [Attenuation_0], properties {pitchMax=1.4, pitchMin=1.3, volumeMax=1, volumeMin=1}}
 {id=Modulator_7, children [Attenuation_1], properties {pitchMax=2, pitchMin=2}}
 {id=QualityLevel_0, children [WavePlayer_0,WavePlayer_0,WavePlayer_9,WavePlayer_9]}
 {id=Random_0, children [WavePlayer_1,WavePlayer_2,WavePlayer_3], properties {weights [1,1,1]}}
 {
  id=SoundClass_0
  children [Delay_1]
  properties {soundClassOverride=/Game/Sounds/FortSoundClassesAndModes/Fort_Player_Construction_NonCritical}
 }
 {id=SoundClass_1, children [Modulator_6], properties {soundClassOverride=/Game/Sounds/FortSoundClassesAndModes/Fort_Player_Construction}}
 {
  id=WavePlayer_0
  properties {bLooping=true, soundWaveAssetPtr=/Game/Sounds/Fort_UI_Sounds/Gameplay/BuildingSounds/Wavs/rock_lev_02_construction_loop_B}
 }
 {id=WavePlayer_1, properties {soundWaveAssetPtr=/Game/Sounds/Fort_UI_Sounds/Gameplay/BuildingSounds/Wavs/Fort_Construction_Done_Wood_01}}
 {id=WavePlayer_2, properties {soundWaveAssetPtr=/Game/Sounds/Fort_UI_Sounds/Gameplay/BuildingSounds/Wavs/Fort_Construction_Done_Wood_02}}
 {id=WavePlayer_3, properties {soundWaveAssetPtr=/Game/Sounds/Fort_UI_Sounds/Gameplay/BuildingSounds/Wavs/Fort_Construction_Done_Wood_03}}
 {
  id=WavePlayer_9
  properties {bLooping=true, soundWaveAssetPtr=/Game/Sounds/Fort_UI_Sounds/Gameplay/BuildingSounds/Wavs/rock_lev_02_construction_loop_B_opt}
 }
]
properties {
 maxDistance=3584
 volumeMultiplier=1
 attenuationOverrides {falloffDistance=2600}
 concurrencyOverrides {resolutionRule=PreventNew}
 concurrencySet [/Game/Sounds/Concurrency_Rules/Player_Wall_Build_Conc]
}
