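# Sound cue for the start of an enemy's level 1 wood build, mixing an initial hit with a looping construction bed.
# The hit randomly picks one of three pitched-up "construction done" wood waves, using enemy build-initiate attenuation and the construction sound class.
# The quieter loop follows a short delay with wall attenuation and a non-critical class, switching by quality level between two random loops and one optimized loop.
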SoundCue /Game/Sounds/Fort_UI_Sounds/Gameplay/BuildingSounds/SoundCues/Fort_Build_StartWood_Lev1_Enemy_Cue
root=Mixer_1
nodes [
 {id=Attenuation_0, children [Random_0], properties {attenuationSettings=/Game/Sounds/Attenuation/Fort_Build_Initiate_Enemy_Att}}
 {id=Attenuation_1, children [Delay_1], properties {attenuationSettings=/Game/Sounds/Attenuation/Fort_Build_Wall_Att}}
 {id=Delay_1, children [Modulator_3], properties {delayMax=0.33, delayMin=0.3}}
 {id=Mixer_1, children [SoundClass_1,SoundClass_0], properties {inputVolume [1,0.6]}}
 {id=Modulator_2, children [Attenuation_0], properties {pitchMax=1.5, pitchMin=1.3, volumeMax=1, volumeMin=1}}
 {id=Modulator_3, children [QualityLevel_0]}
 {id=QualityLevel_0, children [Random_2,Random_2,WavePlayer_7,WavePlayer_7]}
 {id=Random_0, children [WavePlayer_0,WavePlayer_1,WavePlayer_2], properties {weights [1,1,1]}}
 {id=Random_2, children [WavePlayer_3,WavePlayer_6], properties {weights [1,1]}}
 {
  id=SoundClass_0
  children [Attenuation_1]
  properties {soundClassOverride=/Game/Sounds/FortSoundClassesAndModes/Fort_Player_Construction_NonCritical}
 }
 {id=SoundClass_1, children [Modulator_2], properties {soundClassOverride=/Game/Sounds/FortSoundClassesAndModes/Fort_Player_Construction}}
 {id=WavePlayer_0, properties {soundWaveAssetPtr=/Game/Sounds/Fort_UI_Sounds/Gameplay/BuildingSounds/Wavs/Fort_Construction_Done_Wood_01}}
 {id=WavePlayer_1, properties {soundWaveAssetPtr=/Game/Sounds/Fort_UI_Sounds/Gameplay/BuildingSounds/Wavs/Fort_Construction_Done_Wood_02}}
 {id=WavePlayer_2, properties {soundWaveAssetPtr=/Game/Sounds/Fort_UI_Sounds/Gameplay/BuildingSounds/Wavs/Fort_Construction_Done_Wood_03}}
 {
  id=WavePlayer_3
  properties {bLooping=true, soundWaveAssetPtr=/Game/Sounds/Fort_UI_Sounds/Gameplay/BuildingSounds/Wavs/wood_lev_01_construction_loop_A}
 }
 {
  id=WavePlayer_6
  properties {bLooping=true, soundWaveAssetPtr=/Game/Sounds/Fort_UI_Sounds/Gameplay/BuildingSounds/Wavs/wood_lev_01_construction_loop_B}
 }
 {
  id=WavePlayer_7
  properties {bLooping=true, soundWaveAssetPtr=/Game/Sounds/Fort_UI_Sounds/Gameplay/BuildingSounds/Wavs/wood_lev_01_construction_loop_opt}
 }
]
properties {
 maxDistance=3584
 volumeMultiplier=1
 attenuationOverrides {falloffDistance=2500}
 concurrencyOverrides {resolutionRule=PreventNew}
 concurrencySet [/Game/Sounds/Concurrency_Rules/Player_Wall_Build_Conc]
}
